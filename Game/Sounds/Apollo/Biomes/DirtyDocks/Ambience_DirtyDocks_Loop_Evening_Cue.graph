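# Looping evening ambience Sound Cue for the DirtyDocks biome in Apollo.
# A quality-level node at the root picks the looping wave: the quad recording for the first two levels, the stereo version for the last two.
# The cue plays at reduced volume through the Athena outdoor ambient sound class and the Fort ambient Athena attenuation settings.
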SoundCue /Game/Sounds/Apollo/Biomes/DirtyDocks/Ambience_DirtyDocks_Loop_Evening_Cue
root=QualityLevel_0
nodes [
 {id=QualityLevel_0, children [WavePlayer_5,WavePlayer_5,WavePlayer_4,WavePlayer_4]}
 {id=WavePlayer_4, properties {bLooping=true, soundWaveAssetPtr=&/Ambience_DirtyDocks_Evening_Quad_Stereo}}
 {id=WavePlayer_5, properties {bLooping=true, soundWaveAssetPtr=&/Ambience_DirtyDocks_Evening_Quad_Quad}}
]
properties {
 attenuationSettings=&Fort_Ambient_Athena_Att
 maxDistance=2097152
 soundClassObject=&Fort_Ambient_Outdoors_Athena
 volumeMultiplier=0.601
 preEffectBusSends [{}]
}
