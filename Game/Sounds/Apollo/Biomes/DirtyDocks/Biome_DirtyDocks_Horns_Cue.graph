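# Sound cue for one-shot horn sounds in the Dirty Docks biome's outdoor ambience.
# A Random node picks between two horn waves with equal weight, the second lowered in volume through a Mixer.
# A root Modulator varies the pitch, reaching as low as half speed.
# The cue plays quietly with tile-ambient one-shot attenuation and sends to the exterior ambient source bus.
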SoundCue /Game/Sounds/Apollo/Biomes/DirtyDocks/Biome_DirtyDocks_Horns_Cue
root=Modulator_0
nodes [
 {id=Mixer_0, children [WavePlayer_2], properties {inputVolume [0.25]}}
 {id=Modulator_0, children [Random_0], properties {pitchMin=0.5}}
 {id=Random_0, children [WavePlayer_1,Mixer_0], properties {weights [1,1]}}
 {id=WavePlayer_1, properties {soundWaveAssetPtr=&/Biome_DirtyDocks_OneShot_Horn_01}}
 {id=WavePlayer_2, properties {soundWaveAssetPtr=&/Biome_DirtyDocks_OneShot_Horn_02}}
]
properties {
 attenuationSettings=/Game/Sounds/Attenuation/Fort_TileAmbient_OneShot_Att
 maxDistance=8000
 soundClassObject=&Fort_Ambient_Outdoors_Athena
 volumeMultiplier=0.251
 busSends [{audioBus=null, minSendLevel=1, sendLevel=1, soundSourceBus=&AmbientAudio_Exterior_SourceBus}]
}
